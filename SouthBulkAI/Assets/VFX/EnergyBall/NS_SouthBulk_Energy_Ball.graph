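# Niagara system for the SouthBulk energy ball visual effect, layered from Core, Glow, Corona_OneShot and FakeLight emitters.
# A Light emitter is present but disabled.
# System-level logic pairs SystemState with six Set Variables modules, and the system draws on waveform, curve, multiply and float-to-vector dynamic inputs.
# The system exposes no user parameters, and no renderers are listed for any emitter.
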NiagaraSystem /SouthBulkAI/Assets/VFX/EnergyBall/NS_SouthBulk_Energy_Ball
dependencies {
 usedDataInterfaces [&NiagaraDataInterfaceCurve]
 usedDynamicInputs [
  &Multiply_VectorByFloat
  &Waveform
  &FloatFromCurve
  &VectorFromFloat
  &Vector2DFromFloat
  &Multiply_Float
 ]
 usedModules [
  &SystemState
  &self:SystemSpawnScript.SystemScriptSource.SystemScriptGraph.NiagaraNodeAssignment_0.SetVariables_#1
  &self:SystemSpawnScript.SystemScriptSource.SystemScriptGraph.NiagaraNodeAssignment_1.SetVariables_#2
  &self:SystemSpawnScript.SystemScriptSource.SystemScriptGraph.NiagaraNodeAssignment_5.SetVariables_#3
  &self:SystemSpawnScript.SystemScriptSource.SystemScriptGraph.NiagaraNodeAssignment_4.SetVariables_#4
  &self:SystemSpawnScript.SystemScriptSource.SystemScriptGraph.NiagaraNodeAssignment_2.SetVariables_#5
  &self:SystemSpawnScript.SystemScriptSource.SystemScriptGraph.NiagaraNodeAssignment_3.SetVariables_#6
 ]
}
emitters {
 Core {rendererClasses [], renderers []}
 Corona_OneShot {rendererClasses [], renderers []}
 FakeLight {rendererClasses [], renderers []}
 Glow {rendererClasses [], renderers []}
 Light {bEnabled=false, rendererClasses [], renderers []}
}
summary {
 systemName=NS_SouthBulk_Energy_Ball
 emitters [
  {emitterName=Core, rendererClasses []}
  {emitterName=Glow, rendererClasses []}
  {bEnabled=false, emitterName=Light, rendererClasses []}
  {emitterName=Corona_OneShot, rendererClasses []}
  {emitterName=FakeLight, rendererClasses []}
 ]
 userVariables []
}
