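# Niagara impact effect for the SouthBulk energy ball, built from Flare, FlareLong, Glow, DarkGlow, WindGusts_Lightweight and Sparks emitters.
# Only Sparks is detailed here: a delayed GPU burst from a small sphere, slowed by drag, pulled down by gravity and stretched by speed.
# Curves shift spark color and scale size and alpha over life.
# FN_Sprite_Attributes derives an enlarged, separately tinted attribute set alongside two sprite renderers.
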NiagaraSystem /SouthBulkAI/Assets/VFX/EnergyBall/NS_SouthBulk_Energy_Ball_Impact
dependencies {
 usedDataInterfaces [&NiagaraDataInterfaceCurve, &NiagaraDataInterfaceColorCurve]
 usedDynamicInputs [
  &Multiply_VectorByFloat
  &Waveform
  &VectorFromFloat
  &Multiply_Float
 ]
 usedModules [
  &SystemState
  &self:SystemSpawnScript.SystemScriptSource.SystemScriptGraph.NiagaraNodeAssignment_0.SetVariables_#1
  &self:SystemSpawnScript.SystemScriptSource.SystemScriptGraph.NiagaraNodeAssignment_1.SetVariables_#2
  &self:SystemSpawnScript.SystemScriptSource.SystemScriptGraph.NiagaraNodeAssignment_5.SetVariables_#3
  &self:SystemSpawnScript.SystemScriptSource.SystemScriptGraph.NiagaraNodeAssignment_4.SetVariables_#4
  &self:SystemSpawnScript.SystemScriptSource.SystemScriptGraph.NiagaraNodeAssignment_2.SetVariables_#5
  &self:SystemSpawnScript.SystemScriptSource.SystemScriptGraph.NiagaraNodeAssignment_3.SetVariables_#6
 ]
}
emitters {
 DarkGlow {rendererClasses [], renderers []}
 Flare {rendererClasses [], renderers []}
 FlareLong {rendererClasses [], renderers []}
 Glow {rendererClasses [], renderers []}
 Sparks {
  simTarget=GPUComputeSim
  emitterUpdateScript [
   &EmitterState(Life Cycle Mode=::NewEnumerator1, Loop Behavior=::NewEnumerator1, Loop Delay=0.04, Loop Duration=1, UseLoopDelay=true)
   &SpawnBurst_Instantaneous(Spawn Count=10, Use Spawn Probability=false, UseLoopCountLimit=true)
  ]
  particleSpawnScript [
   &InitializeParticle(
    Color=<rgba 2 2 2 1>
    Lifetime Max=0.6
    Lifetime Min=0.2
    Lifetime Mode=::NewEnumerator1
    Sprite Size Mode=::NewEnumerator3
    Uniform Sprite Size=&RandomRangeFloat(Minimum=1, Maximum=3)
    UsePositionOffset=false
    Write Material Random=true
   )
   &CameraOffset(Camera Offset Amount=25)
   &ShapeLocation(MaskSpawnGroup=false, Override Seed=false, Sphere Radius=20, Sphere Surface Distribution=0.5357, UseOwnerScale=true)
   &DynamicMaterialParameters(
    Index 0 Param 1=2
    Index 0 Param 2=10
    Index 0 Param 3=1
    Module.Index0Float0Write=true
    Module.Index0Float1Write=true
    Module.Index0Float2Write=true
    Module.Index0Float3Write=false
   )
   &AddVelocity(
    Module.Use Falloff Distance=false
    Override Seed=false
    Velocity Mode=::NewEnumerator1
    Velocity Speed=&RandomRangeFloat(Minimum=120, Maximum=400)
   )
  ]
  particleUpdateScript [
   &ParticleState()
   &Drag(Drag=&RandomRangeFloat(Minimum=1.6, Maximum=4), Use Linear Drag=true, Use Rotational Drag=false)
   &GravityForce(Gravity=<0 0 -200>)
   &SolveForcesAndVelocity(Clamp Velocity=false, Limit Acceleration=false)
   &ScaleSpriteSizeBySpeed(Max Scale Factor=<0.5 8>, Min Scale Factor=<1 0.5>, Velocity Threshold=600)
   &ScaleSpriteSize(Scale Factor=&Vector2DFromFloat(Value=&FloatFromCurve(FloatCurve=DataInterface() {propertyValues {Curve {keys [
          {arriveTangent=3.8859, interpMode=RCIM_Cubic, leaveTangent=3.8859, tangentMode=RCTM_User, time=0, value=0}
          {arriveTangent=0.1164, interpMode=RCIM_Cubic, leaveTangent=0.1164, tangentMode=RCTM_User, time=0.2, value=1}
          {arriveTangent=-0.2479, interpMode=RCIM_Cubic, leaveTangent=-0.2479, tangentMode=RCTM_User, time=0.5, value=1}
          {arriveTangent=-1.9959, interpMode=RCIM_Cubic, leaveTangent=-1.9959, tangentMode=RCTM_User, time=1, value=0}
         ]}}})))
   &ScaleColor(
    Scale Alpha=&FloatFromCurve(FloatCurve=DataInterface() {propertyValues {Curve {keys [
         <at 0 0 RCIM_Cubic>
         {arriveTangent=0.159, interpMode=RCIM_Cubic, leaveTangent=0.159, tangentMode=RCTM_User, time=0.1, value=1}
         {interpMode=RCIM_Cubic, tangentMode=RCTM_Break, time=0.85, value=1}
         {interpMode=RCIM_Cubic, tangentMode=RCTM_Break, time=1, value=0}
        ]}}})
    Scale RGB=&MakeVectorFromLinearColorRGB(Color=&ColorFromCurve(ColorCurve=DataInterface() {propertyValues {
        AlphaCurve {keys [<at 0 1 RCIM_Cubic>]}
        BlueCurve {keys [<at 0.1023 0.2 RCIM_Cubic>,<at 0.9955 0>]}
        GreenCurve {keys [<at 0.1023 0.8 RCIM_Cubic>,<at 0.9955 0.1>]}
        RedCurve {keys [<at 0.1023 2 RCIM_Cubic>,<at 0.9955 1>]}
       }}))
    ScaleA=false
    ScaleRGB=true
   )
   &FN_Sprite_Attributes(
    Color=&ColorFromCurve(ColorCurve=DataInterface() {propertyValues {
       AlphaCurve {keys [<at 0.0151 0>,<at 0.0946 0.2 RCIM_Cubic>,<at 0.7029 0 RCIM_Cubic>]}
       BlueCurve {keys [<at 0.0598 0 RCIM_Cubic>,<at 0.7123 0>]}
       GreenCurve {keys [<at 0.0598 0.1667 RCIM_Cubic>,<at 0.7123 0.0833>]}
       RedCurve {keys [<at 0.0598 1 RCIM_Cubic>,<at 0.7123 1>]}
      }})
    Create Color Attribute=true
    Create Dynamic Material Parameter Attributes=true
    Create SpriteSize Attribute=true
    Index 0 Param 1=1
    Index 0 Param 2=10
    Index 0 Param 3=1
    Module Output:ENiagara_AttributeSamplingApplyOutput=::NewEnumerator1
    Module.Index0Float0Write=true
    Module.Index0Float1Write=true
    Module.Index0Float2Write=true
    Module.Index0Float3Write=true
    SpriteSize=&Multiply_Vector2D(A=$StackContext.SpriteSize, B=<6 2.6>)
   )
  ]
  rendererClasses [&NiagaraSpriteRendererProperties]
  renderers [
   {rendererClass=&NiagaraSpriteRendererProperties}
   {rendererClass=&NiagaraSpriteRendererProperties, rendererIndex=1}
  ]
 }
 WindGusts_Lightweight {rendererClasses [], renderers []}
}
summary {
 systemName=NS_SouthBulk_Energy_Ball_Impact
 emitters [
  {emitterName=Flare, rendererClasses []}
  {emitterName=DarkGlow, rendererClasses []}
  {emitterName=WindGusts_Lightweight, rendererClasses []}
  {emitterName=FlareLong, rendererClasses []}
  {emitterName=Glow, rendererClasses []}
  {emitterName=Sparks, simTarget=GPUComputeSim, rendererClasses [&NiagaraSpriteRendererProperties]}
 ]
 userVariables []
}
